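# Sound cue for the looping ambience of a Rift Beacon bubble in Athena, named for the bubble's center as heard from outside.
# A Mixer feeds the output from a single looping wave player of a distant rift gravity loop, with its other input left empty.
# The cue plays at reduced volume under the scripted events sound class, using an outside-ambience attenuation asset for spatial falloff.
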SoundCue /Game/Athena/Sounds/Events/RiftBeacon/RiftBeacon_Bubble_Center_Outside_Loop_Cue
root=Mixer_2
nodes [
 {id=Mixer_2, children [null,WavePlayer_1], properties {inputVolume [1,1]}}
 {id=WavePlayer_1, properties {bLooping=true, soundWaveAssetPtr=/Game/Athena/Sounds/Events/RiftBeacon/AssetDrop_01/Rift_GravityLoop_Dist_01}}
]
properties {
 attenuationSettings=&/RiftBeacon_Bubble_Amb_Outside_Att
 maxDistance=1000000
 soundClassObject=/Game/Sounds/FortSoundClassesAndModes/Fort_Events_Scripted
 volumeMultiplier=0.2632
}
