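# Sound cue for the looping sound at the edge of the Rift Beacon bubble, routed to the scripted events sound class.
# A mixer blends two looping waves, a close rift gravity loop and a reused hoverboard movement loop, with the gravity loop weighted louder.
# Each layer has its own attenuation override with wide stereo spread.
# The hoverboard layer has a shorter falloff and enables distance-based filtering.
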SoundCue /Game/Athena/Sounds/Events/RiftBeacon/RiftBeacon_Bubble_Edge_Loop_Cue
root=Mixer_1
nodes [
 {
  id=Attenuation_0
  children [WavePlayer_0]
  properties {bOverrideAttenuation=true, attenuationOverrides {falloffDistance=1200, stereoSpread=512, attenuationShapeExtents {x=0}}}
 }
 {
  id=Attenuation_1
  children [WavePlayer_3]
  properties {
   bOverrideAttenuation=true
   attenuationOverrides {
    bAttenuateWithLPF=true
    falloffDistance=1000
    hPFFrequencyAtMax=15000
    hPFFrequencyAtMin=15000
    lPFRadiusMax=1000
    lPFRadiusMin=0
    stereoSpread=512
    attenuationShapeExtents {x=0}
   }
  }
 }
 {id=Mixer_1, children [Attenuation_0, Attenuation_1], properties {inputVolume [1.5,1]}}
 {id=WavePlayer_0, properties {bLooping=true, soundWaveAssetPtr=/Game/Athena/Sounds/Events/RiftBeacon/AssetDrop_01/Rift_GravityLoop_Close_01}}
 {id=WavePlayer_3, properties {bLooping=true, soundWaveAssetPtr=/Game/Athena/Sounds/Vehicles/Hoverboard/Hoverboard_MoveLoop_06}}
]
properties {maxDistance=1200, soundClassObject=/Game/Sounds/FortSoundClassesAndModes/Fort_Events_Scripted, volumeMultiplier=0.3509}
